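# Translucent unlit fog material for Niagara mesh particles, named for the S17 Hatchery and colored by particle color.
# Opacity multiplies two panning smoke samples by a UV edge mask, vertex red, particle alpha and a Fresnel term.
# A depth fade softens intersections, its distance varied by a third panning smoke texture.
# An animated, actor-relative vector noise with Curl parameters feeds a shading-path switch that uses a constant on mobile.
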Material /Game/Environments/Apollo/Props/Quests/Props/FX/M_S17_HatcheryFog
graph {
 properties {
  bUsedWithNiagaraMeshParticles=true
  blendMode=BLEND_Translucent
  pixelDepthOffsetMode=PDOM_Legacy
  shadingModel=MSM_Unlit
  usedShadingModels=Unlit
 }
 nodes [
  n1 = ParticleColor()
  n2 = TextureCoordinate()
  n3 = ComponentMask(None=@n2) {g=true}
  n4 = Clamp(None=Multiply(A=@n3) {constB=2})
  n5 = Clamp(None=Multiply(A=OneMinus(None=@n3)) {constB=2})
  n6 = ComponentMask(None=@n2) {r=true}
  n7 = Clamp(None=Multiply(A=@n6) {constB=2})
  n8 = Clamp(None=Multiply(A=OneMinus(None=@n6)) {constB=2})
  n9 = Power(Base=Multiply(A=Multiply(A=@n4, B=@n5), B=Multiply(A=@n7, B=@n8))) {constExponent=1.3}
  n10 = TextureCoordinate()
  n11 = AppendVector(A=ScalarParameter() {defaultValue=0.4, parameterName=U1}, B=ScalarParameter() {defaultValue=0.4, parameterName=V1})
  n12 = Multiply(A=Time(), B=ScalarParameter() {defaultValue=0.2, parameterName=Panner Speed})
  n13 = TextureSample(UVs=Panner(Coordinate=Multiply(A=@n10, B=@n11), Time=@n12) {bFractionalPart=true, speedX=0.0121, speedY=0.0534}) {texture=/Game/Global/Textures/Default/Tileable_Patterns/T_Smoke_Puff_Tile_Up}
  n14 = AppendVector(A=ScalarParameter() {defaultValue=0.38, parameterName=U2}, B=ScalarParameter() {defaultValue=0.22, parameterName=V2})
  n15 = TextureSample(UVs=Panner(Coordinate=Multiply(A=@n10, B=@n14), Time=@n12) {bFractionalPart=true, speedX=-0.0155, speedY=0.05}) {texture=&T_Smoke_Puff_Tile_Up}
  n16 = Power(Base=Divide(A=Add(A=@n13.RGB, B=@n15.RGB)) {constB=2.2}, Exp=ScalarParameter() {defaultValue=2.2, parameterName=Cloud Exponent}) {constExponent=3}
  n17 = Multiply(A=@n1.A, B=Multiply(A=VertexColor().R, B=Multiply(A=@n9, B=@n16)))
  n18 = MaterialFunctionCall(
   Brightness=ScalarParameter() {defaultValue=0.9, parameterName=Fresnel Brightness}
   Exponent=ScalarParameter() {defaultValue=1.4, parameterName=Fresnel Exponent}
  ) {materialFunction=&MF_Fresnel_FX}
  n19 = Panner(Coordinate=TextureCoordinate() {uTiling=3, vTiling=3}) {bFractionalPart=true, speedX=0.01, speedY=0.03}
  n20 = Power(Base=Divide(A=TextureSample(UVs=@n19) {samplerType=SAMPLERTYPE_LinearColor, texture=/Game/Global/Textures/Default/Tileable_Patterns/T_Tile_Smoke_Small}.R) {constB=0.9})
  DepthFade(
   Opacity=Clamp(None=Multiply(A=@n17, B=@n18.Fresnel))
   FadeDistance=LinearInterpolate(
    A=ScalarParameter() {defaultValue=10, parameterName=Depth Fade Distance2}
    B=ScalarParameter() {defaultValue=30, parameterName=Depth Fade Distance1}
    Alpha=Clamp(None=@n20)
   )
  )
  MaterialFunctionCall(Emissive=@n1.RGB) {materialFunction=&MF_FX_EmissiveAdjustments}
  n21 = WorldPosition()
  n22 = ActorPositionWS()
  n23 = ScalarParameter() {defaultValue=10, parameterName=Curl_Speed}
  n24 = Time()
  n25 = Divide(A=Subtract(A=Subtract(A=@n21.XYZ, B=@n22), B=Multiply(A=@n23, B=@n24)), B=ScalarParameter() {defaultValue=200, parameterName=Curl_Scale})
  ShadingPathSwitch(
   Default=Multiply(A=ScalarParameter() {defaultValue=10, parameterName=Curl_Intensity}, B=VectorNoise(World Position=@n25) {noiseFunction=VNF_VectorALU})
   Mobile=Constant() {r=1}
  ) {desc=Migrated from FeatureLevelSwitch}
 ]
}
