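# Niagara fog effect for the Season 17 Alien Hatchery quest prop, built from two smoke emitters.
# Smoke_Plane bursts one large, persistent mesh particle with SubUV animation and a slight random camera offset.
# Swirly_Smoke continuously spawns faint sprites across a flat box that grow, slowly rise, and fade in and out, varied by random SubUV frame ranges, rotation rates and dynamic material parameters.
# It declares a User.BP_On_Off bool, which no shown module reads.
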NiagaraSystem /Game/Environments/Apollo/Props/Quests/Props/FX/NS_S17_AlienHatchery_Fog
dependencies {usedDataInterfaces [&NiagaraDataInterfaceCurve], usedDynamicInputs [], usedModules [&SystemState]}
emitters {
 Smoke_Plane {
  emitterUpdateScript [
   &EmitterState(Life Cycle Mode=::NewEnumerator1, Loop Behavior=::NewEnumerator1, UseLoopDelay=true)
   &SpawnBurst_Instantaneous(Spawn Count=1, Use Spawn Probability=false)
  ]
  particleSpawnScript [
   &InitializeParticle(
    Color=<rgba 0.8229 0.9171 1 2>
    Lifetime=4
    Mesh Scale Mode=::NewEnumerator3
    Mesh Uniform Scale=500
    Sprite Rotation Mode=::NewEnumerator2
    Sprite Size Mode=::NewEnumerator3
    Uniform Sprite Size=25
    UsePositionOffset=false
    Write Material Random=true
   ) {module=InitializeParticle001}
   &CameraOffset(Camera Offset Amount=&RandomRangeFloat(Minimum=-2, Maximum=1))
  ]
  particleUpdateScript [
   &ParticleState(Kill Particles When Lifetime Has Elapsed=false)
   /Niagara/Modules/Update/SubUV/SubUVAnimation(Number Of Frames=62)
  ]
  rendererClasses [&NiagaraMeshRendererProperties]
  renderers [{rendererClass=&NiagaraMeshRendererProperties, rendererIndex=0}]
 }
 Swirly_Smoke {
  emitterUpdateScript [
   &EmitterState(
    Inactive Response=::NewEnumerator1
    Life Cycle Mode=::NewEnumerator1
    Loop Duration=&RandomRangeFloat(Minimum=0.35, Maximum=0.75, Evaluation Type=::NewEnumerator1)
    Recalculate Duration Each Loop=true
    UseLoopDelay=true
   )
   &SpawnRate(SpawnRate=3, Use Spawn Probability=false)
   &SpawnBurst_Instantaneous(Spawn Count=&RandomRangeInt(Minimum=7, Maximum=10), Use Spawn Probability=false) {enabled=false}
  ]
  particleSpawnScript [
   /SetVariables(Particles.SpriteRotationRate=&RandomRangeFloat(Minimum=-0.5, Maximum=-0.3)) {isSetParameters=true}
   &InitializeParticle(
    Color=<rgba 1 1 1 0.5>
    Lifetime=6
    Sprite Rotation Mode=::NewEnumerator2
    Sprite Size Mode=::NewEnumerator1
    Uniform Sprite Size Max=200
    Uniform Sprite Size Min=100
    UsePositionOffset=false
    Write Material Random=true
   )
   &DynamicMaterialParameters(
    Index 0 Param 1=&RandomRangeFloat(Minimum=0, Maximum=1, Evaluation Type=::NewEnumerator1)
    Index 0 Param 2=&RandomRangeFloat(Minimum=2, Maximum=4, Evaluation Type=::NewEnumerator1)
    Module.Param0WriteEnabled=true
    Module.Param1WriteEnabled=true
    Module.Param2WriteEnabled=false
    Module.Param3WriteEnabled=false
   )
   &BoxLocation(
    Box Size=&RandomRangeVector(Minimum=<500 500 0>, Maximum=<-500 -500 1>)
    MaskSpawnGroup=false
    Module.Override Seed=false
    Rotate Box=false
    Surface Only=false
   )
  ]
  particleUpdateScript [
   &ParticleState()
   &ScaleColor(
    Scale Alpha=&FloatFromCurve(
     FloatCurve=DataInterface() {propertyValues {Curve {keys [
         {arriveTangent=-0.0412, interpMode=RCIM_Cubic, leaveTangent=-0.0412, tangentMode=RCTM_User, time=0, value=0}
         {arriveTangent=0.0848, interpMode=RCIM_Cubic, leaveTangent=0.0848, tangentMode=RCTM_User, time=0.4, value=0.1}
         {arriveTangent=-0.0405, interpMode=RCIM_Cubic, leaveTangent=-0.0405, tangentMode=RCTM_User, time=1, value=0}
        ]}}}
     Scale Curve=&MakeCustomFloatFromBool(Bool=false, True Float=1, False Float=2)
    )
    ScaleA=true
    ScaleRGB=true
   )
   /Niagara/Modules/Update/SubUV/V2/SubUVAnimation(
    End Frame:NiagaraInt32=&RandomRangeInt(Minimum=35, Maximum=64)
    Start Frame:NiagaraInt32=&RandomRangeInt(Minimum=0, Maximum=28)
   ) {module=SubUVAnimation001}
   &DynamicMaterialParameters(
    Index 0 Param 3=0.3
    Module.Param0WriteEnabled=false
    Module.Param1WriteEnabled=false
    Module.Param2WriteEnabled=true
    Module.Param3WriteEnabled=true
   ) {module=DynamicMaterialParameters001}
   &ScaleSpriteSize(Scale Factor=&Vector2DFromFloat(Value=&FloatFromCurve(FloatCurve=DataInterface() {propertyValues {Curve {keys [
          {arriveTangent=4.1412, interpMode=RCIM_Cubic, leaveTangent=4.1412, tangentMode=RCTM_User, time=0, value=0.2}
          {arriveTangent=0.4981, interpMode=RCIM_Cubic, leaveTangent=0.4981, time=0.1, value=0.55}
          {interpMode=RCIM_Cubic, leaveTangent=0.3663, time=1.0038, value=1}
         ]}}})))
   &GravityForce(Gravity=&RandomRangeVector(Minimum=<0 0 6>, Maximum=<0 0 10>))
   &SolveForcesAndVelocity(Clamp Velocity=false, Limit Acceleration=false)
  ]
  rendererClasses [&NiagaraSpriteRendererProperties]
  renderers [{rendererClass=&NiagaraSpriteRendererProperties}]
 }
}
summary {
 systemName=NS_S17_AlienHatchery_Fog
 emitters [
  {emitterName=Smoke_Plane, rendererClasses [&NiagaraMeshRendererProperties]}
  {emitterName=Swirly_Smoke, rendererClasses [&NiagaraSpriteRendererProperties]}
 ]
 userVariables [User.BP_On_Off {type=NiagaraBool}]
}
